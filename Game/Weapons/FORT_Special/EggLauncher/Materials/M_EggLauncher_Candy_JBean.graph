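# Subsurface-shaded material for Egg Launcher jelly bean candy, enabled for mesh particles and Niagara mesh particles.
# Particle color provides per-particle tinting, also multiplied by an SSColor parameter and a Fresnel term with adjustable exponent for an edge-weighted tint.
# Scalar parameters expose metallic, specular, roughness and inner subsurface opacity, with no texture sampling.
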Material /Game/Weapons/FORT_Special/EggLauncher/Materials/M_EggLauncher_Candy_JBean
graph {
 properties {
  bScreenSpaceReflections=true
  bUseTranslucencyVertexFog=false
  bUsedWithMeshParticles=true
  bUsedWithNiagaraMeshParticles=true
  pixelDepthOffsetMode=PDOM_Legacy
  shadingModel=MSM_Subsurface
  translucencyLightingMode=TLM_VolumetricDirectional
  translucencyPass=MTP_BeforeDOF
  translucentSelfShadowDensityScale=1
  translucentSelfShadowSecondDensityScale=1
  translucentSelfShadowSecondOpacity=5
  translucentShadowDensityScale=1
  usedShadingModels=Subsurface
 }
 nodes [
  Multiply(
   A=Multiply(A=ParticleColor().RGB, B=VectorParameter() {parameterName=SSColor, defaultValue=<rgba 0.1 0.1 0.1 1>}.RGB)
   B=Fresnel(ExponentIn=ScalarParameter() {defaultValue=1, group=SS, parameterName=SSFresnel}) {baseReflectFraction=0.005, exponent=8}
  )
  ParticleColor()
  ScalarParameter() {parameterName=Metal}
  ScalarParameter() {defaultValue=0.3, group=SS, parameterName=SSOpacityInner}
  ScalarParameter() {defaultValue=0.8, parameterName=Spec}
  ScalarParameter() {defaultValue=0.45, parameterName=Rough}
 ]
}
